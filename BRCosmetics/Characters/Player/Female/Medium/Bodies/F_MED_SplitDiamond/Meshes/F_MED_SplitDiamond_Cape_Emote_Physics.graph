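# Physics asset for the emote variant of the SplitDiamond female medium character's cape.
# Torso and left lower-arm shapes form the body volume; upper-arm and right lower-arm bodies have no shapes.
# Four box-and-capsule cape chains hang from the clavicle cape bones and cape_root, with limited swing and twist and locked Swing2 near the top.
# Free constraints tie each side's outer and middle chain tips together and link the body bones.
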PhysicsAsset /BRCosmetics/Characters/Player/Female/Medium/Bodies/F_MED_SplitDiamond/Meshes/F_MED_SplitDiamond_Cape_Emote_Physics
bodies {
 cape_01_l [{shapeType=Capsule, length=1, radius=1, shapeName=cape_01_l_capsule}]
 cape_01_r [{shapeType=Capsule, length=1, radius=1, shapeName=cape_01_r_capsule}]
 cape_02_l [{shapeType=Box, extentX=24.5625, extentY=3.1875, extentZ=16.5869, center=<11.5978 -2 0>}]
 cape_02_r [{shapeType=Box, extentX=24.5625, extentY=3.1875, extentZ=29.0244, center=<11.5978 -2 0>}]
 cape_03_l [{shapeType=Capsule, length=11.5, radius=5.25, shapeName=cape_05_r_capsule, center=<7.3712 0 0>}]
 cape_03_r [{shapeType=Capsule, length=18.25, radius=5.25, shapeName=cape_05_l_capsule, center=<7.3712 0 0>}]
 cape_04_l [{shapeType=Capsule, length=18.75, radius=4.75, shapeName=cape_05_r_capsule, center=<5 0 0>}]
 cape_04_r [{shapeType=Capsule, length=18.75, radius=4.75, shapeName=cape_05_l_capsule, center=<5 0 0>}]
 cape_mid_01_l [{shapeType=Box, extentX=12.5625, extentY=0.6875, extentZ=18.5244, center=<11.5978 -2 0>}]
 cape_mid_01_r [{shapeType=Box, extentX=12.5625, extentY=2.1875, extentZ=18.5244, center=<11.5978 -2 0>}]
 cape_mid_02_l [{shapeType=Box, extentX=24.5625, extentY=2.1875, extentZ=17.1494, center=<11.5978 -2 0>}]
 cape_mid_02_r [{shapeType=Box, extentX=24.5625, extentY=2.1875, extentZ=37.5244, center=<11.5978 -2 0>}]
 cape_mid_03_l [{shapeType=Capsule, length=18.25, radius=5.25, shapeName=cape_05_r_capsule, center=<7.3712 0 0>}]
 cape_mid_03_r [{shapeType=Capsule, length=18.25, radius=5.25, shapeName=cape_05_r_capsule, center=<7.3712 0 0>}]
 cape_mid_04_l [{shapeType=Capsule, length=18.75, radius=4.75, shapeName=cape_05_r_capsule, center=<5 0 0>}]
 cape_mid_04_r [{shapeType=Capsule, length=18.75, radius=4.75, shapeName=cape_05_r_capsule, center=<5 0 0>}]
 cape_root [{shapeType=Capsule, length=0.505, radius=0.505, shapeName=cape_root_capsule, rotation=<rot 0 -90 0>}]
 clavicle_cape_l [{shapeType=Capsule, length=1, radius=1, shapeName=clavicle_cape_l_capsule}]
 clavicle_cape_r [{shapeType=Capsule, length=1, radius=1, shapeName=clavicle_cape_r_capsule}]
 lowerarm_l [{
   shapeType=Capsule
   length=15.375
   radius=4.6562
   shapeName=lowerarm_l_capsule
   center=<12.4621 0 -0.3303>
   rotation=<rot -89.8 0.0008 -0.0008>
  }]
 lowerarm_r []
 pelvis [{shapeType=Capsule, length=11.4028, radius=8.6596, center=<-1.9544 -2.2029 -0.1981>}]
 spine_01 [{shapeType=Capsule, length=13.7671, radius=6.7354, center=<-2.8183 -1.0873 0>}]
 spine_02 [{shapeType=Capsule, length=4.75, radius=8, shapeName=spine_02_capsule, center=<0 1.6712 0>}]
 spine_03 [
  {shapeType=Capsule, length=3, radius=10.7354, center=<1.9324 0.5037 0>}
  {shapeType=Capsule, length=20.7188, radius=9.2812, center=<-1.2326 2.6059 0>, rotation=<rot 90 0 0>}
 ]
 spine_04 [
  {shapeType=Capsule, length=19.0796, radius=7.3916, center=<6.9796 -1.0098 0>}
  {shapeType=Capsule, length=10.0312, radius=9.2812, center=<-1.2326 1.308 0>, rotation=<rot 90 0 0>}
 ]
 spine_05 [{shapeType=Capsule, length=13.7359, radius=7.3291, center=<1.5198 -0.8941 0>}]
 upperarm_l []
 upperarm_r []
}
constraints [
 {
  bone1Name=spine_03
  bone2Name=spine_02
  swing1LimitDegrees=45
  swing1Motion=Free
  swing2LimitDegrees=45
  swing2Motion=Free
  twistLimitDegrees=45
  twistMotion=Free
 }
 {
  bone1Name=spine_05
  bone2Name=spine_04
  swing1LimitDegrees=45
  swing1Motion=Free
  swing2LimitDegrees=45
  swing2Motion=Free
  twistLimitDegrees=45
  twistMotion=Free
 }
 {
  bone1Name=spine_04
  bone2Name=spine_03
  swing1LimitDegrees=45
  swing1Motion=Free
  swing2LimitDegrees=45
  swing2Motion=Free
  twistLimitDegrees=45
  twistMotion=Free
 }
 {
  bone1Name=cape_root
  bone2Name=spine_05
  swing1LimitDegrees=45
  swing1Motion=Free
  swing2LimitDegrees=45
  swing2Motion=Free
  twistLimitDegrees=45
  twistMotion=Free
 }
 {bone1Name=cape_mid_02_l, bone2Name=cape_mid_01_l, swing1LimitDegrees=30, swing2LimitDegrees=5, swing2Motion=Locked, twistLimitDegrees=5}
 {bone1Name=cape_mid_03_l, bone2Name=cape_mid_02_l, swing1LimitDegrees=35, swing2LimitDegrees=15, twistLimitDegrees=5}
 {bone1Name=cape_mid_04_l, bone2Name=cape_mid_03_l, swing1LimitDegrees=35, swing2LimitDegrees=15, twistLimitDegrees=5}
 {bone1Name=cape_mid_02_r, bone2Name=cape_mid_01_r, swing1LimitDegrees=30, swing2LimitDegrees=5, swing2Motion=Locked, twistLimitDegrees=5}
 {bone1Name=cape_mid_03_r, bone2Name=cape_mid_02_r, swing1LimitDegrees=35, swing2LimitDegrees=15, twistLimitDegrees=5}
 {bone1Name=cape_mid_04_r, bone2Name=cape_mid_03_r, swing1LimitDegrees=35, swing2LimitDegrees=15, twistLimitDegrees=5}
 {
  bone1Name=clavicle_cape_l
  bone2Name=spine_05
  swing1LimitDegrees=45
  swing1Motion=Free
  swing2LimitDegrees=45
  swing2Motion=Free
  twistLimitDegrees=45
  twistMotion=Free
 }
 {bone1Name=cape_02_l, bone2Name=cape_01_l, swing1LimitDegrees=30, swing2LimitDegrees=5, swing2Motion=Locked, twistLimitDegrees=5}
 {bone1Name=cape_03_l, bone2Name=cape_02_l, swing1LimitDegrees=35, swing2LimitDegrees=15, twistLimitDegrees=5}
 {bone1Name=cape_04_l, bone2Name=cape_03_l, swing1LimitDegrees=35, swing2LimitDegrees=15, twistLimitDegrees=5}
 {
  bone1Name=clavicle_cape_r
  bone2Name=spine_05
  swing1LimitDegrees=45
  swing1Motion=Free
  swing2LimitDegrees=45
  swing2Motion=Free
  twistLimitDegrees=45
  twistMotion=Free
 }
 {bone1Name=cape_02_r, bone2Name=cape_01_r, swing1LimitDegrees=30, swing2LimitDegrees=5, swing2Motion=Locked, twistLimitDegrees=5}
 {bone1Name=cape_03_r, bone2Name=cape_02_r, swing1LimitDegrees=35, swing2LimitDegrees=15, twistLimitDegrees=5}
 {bone1Name=cape_04_r, bone2Name=cape_03_r, swing1LimitDegrees=35, swing2LimitDegrees=15, twistLimitDegrees=5}
 {
  bone1Name=spine_02
  bone2Name=spine_01
  swing1LimitDegrees=45
  swing1Motion=Free
  swing2LimitDegrees=45
  swing2Motion=Free
  twistLimitDegrees=45
  twistMotion=Free
 }
 {
  bone1Name=spine_01
  bone2Name=pelvis
  swing1LimitDegrees=45
  swing1Motion=Free
  swing2LimitDegrees=45
  swing2Motion=Free
  twistLimitDegrees=45
  twistMotion=Free
 }
 {
  bone1Name=upperarm_l
  bone2Name=spine_05
  swing1LimitDegrees=45
  swing1Motion=Free
  swing2LimitDegrees=45
  swing2Motion=Free
  twistLimitDegrees=45
  twistMotion=Free
 }
 {
  bone1Name=lowerarm_l
  bone2Name=upperarm_l
  swing1LimitDegrees=45
  swing1Motion=Free
  swing2LimitDegrees=45
  swing2Motion=Free
  twistLimitDegrees=45
  twistMotion=Free
 }
 {
  bone1Name=upperarm_r
  bone2Name=spine_05
  swing1LimitDegrees=45
  swing1Motion=Free
  swing2LimitDegrees=45
  swing2Motion=Free
  twistLimitDegrees=45
  twistMotion=Free
 }
 {
  bone1Name=lowerarm_r
  bone2Name=upperarm_r
  swing1LimitDegrees=45
  swing1Motion=Free
  swing2LimitDegrees=45
  swing2Motion=Free
  twistLimitDegrees=45
  twistMotion=Free
 }
 {
  bone1Name=cape_mid_04_r
  bone2Name=cape_04_r
  swing1LimitDegrees=45
  swing1Motion=Free
  swing2LimitDegrees=45
  swing2Motion=Free
  twistLimitDegrees=45
  twistMotion=Free
 }
 {
  bone1Name=cape_mid_04_l
  bone2Name=cape_04_l
  swing1LimitDegrees=45
  swing1Motion=Free
  swing2LimitDegrees=45
  swing2Motion=Free
  twistLimitDegrees=45
  twistMotion=Free
 }
 {
  bone1Name=cape_01_l
  bone2Name=clavicle_cape_l
  swing1LimitDegrees=45
  swing1Motion=Free
  swing2LimitDegrees=45
  swing2Motion=Free
  twistLimitDegrees=45
  twistMotion=Free
 }
 {
  bone1Name=cape_01_r
  bone2Name=clavicle_cape_r
  swing1LimitDegrees=45
  swing1Motion=Free
  swing2LimitDegrees=45
  swing2Motion=Free
  twistLimitDegrees=45
  twistMotion=Free
 }
 {bone1Name=cape_mid_01_l, bone2Name=cape_root, swing1LimitDegrees=10, swing2LimitDegrees=10, swing2Motion=Locked, twistLimitDegrees=5}
 {bone1Name=cape_mid_01_r, bone2Name=cape_root, swing1LimitDegrees=10, swing2LimitDegrees=10, swing2Motion=Locked, twistLimitDegrees=5}
]
