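# Physics asset for the NyxGlass cape on the female medium SplitDiamond body, defining its collision shapes and joints.
# Four four-bone chains, outer pairs from the clavicle cape bones and middle pairs from cape_root, use boxes high and capsules low, with tight angular limits along each chain.
# Spine, pelvis and clavicle capsules provide body collision, while arm bodies have no shapes.
# Body joints, middle-chain roots and same-side links between cape tips rotate freely.
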PhysicsAsset /BRCosmetics/Characters/Player/Female/Medium/Bodies/F_MED_SplitDiamond/Meshes/F_MED_SplitDiamond_Cape_NyxGlass_Physics
bodies {
 cape_01_l [{shapeType=Box, extentX=22.5625, extentY=2.6875, extentZ=29.0244, center=<11.5978 0.0001 0>}]
 cape_01_r [{shapeType=Box, extentX=22.5625, extentY=2.6875, extentZ=29.0244, center=<11.5978 0.0001 0>}]
 cape_02_l [{shapeType=Box, extentX=22.5625, extentY=2.6875, extentZ=29.0244, center=<11.5978 0.0001 0>}]
 cape_02_r [{shapeType=Box, extentX=22.5625, extentY=2.6875, extentZ=29.0244, center=<11.5978 0.0001 0>}]
 cape_03_l [{shapeType=Capsule, length=21.75, radius=7, shapeName=cape_05_r_capsule, center=<5 0 0>}]
 cape_03_r [{shapeType=Capsule, length=21.75, radius=7, shapeName=cape_05_l_capsule, center=<5 0 0>}]
 cape_04_l [{shapeType=Capsule, length=21.75, radius=7, shapeName=cape_05_r_capsule, center=<5 0 0>}]
 cape_04_r [{shapeType=Capsule, length=21.75, radius=7, shapeName=cape_05_l_capsule, center=<5 0 0>}]
 cape_mid_01_l [{shapeType=Box, extentX=22.5625, extentY=2.6875, extentZ=29.0244, center=<11.5978 0.0001 0>}]
 cape_mid_01_r [{shapeType=Box, extentX=22.5625, extentY=2.6875, extentZ=29.0244, center=<11.5978 0.0001 0>}]
 cape_mid_02_l [{shapeType=Box, extentX=22.5625, extentY=2.6875, extentZ=29.0244, center=<11.5978 0.0001 0>}]
 cape_mid_02_r [{shapeType=Box, extentX=22.5625, extentY=2.6875, extentZ=29.0244, center=<11.5978 0.0001 0>}]
 cape_mid_03_l [{shapeType=Capsule, length=21.75, radius=7, shapeName=cape_05_r_capsule, center=<5 0 0>}]
 cape_mid_03_r [{shapeType=Capsule, length=21.75, radius=7, shapeName=cape_05_r_capsule, center=<5 0 0>}]
 cape_mid_04_l [{shapeType=Capsule, length=21.75, radius=7, shapeName=cape_05_r_capsule, center=<5 0 0>}]
 cape_mid_04_r [{shapeType=Capsule, length=21.75, radius=7, shapeName=cape_05_r_capsule, center=<5 0 0>}]
 cape_root [{shapeType=Capsule, length=0.505, radius=0.505, shapeName=cape_root_capsule, rotation=<rot 0 -90 0>}]
 clavicle_cape_l [{
   shapeType=Capsule
   length=5.8379
   radius=4.5244
   shapeName=clavicle_cape_l_capsule
   center=<8.7474 -3.3449 5.1334>
   rotation=<rot 29.5047 -90 -22.8114>
  }]
 clavicle_cape_r [{
   shapeType=Capsule
   length=6.9146
   radius=4.897
   shapeName=clavicle_cape_r_capsule
   center=<-8.1625 6.0686 -5.1183>
   rotation=<rot 19.2632 -90 -34.7934>
  }]
 clavicle_l [{
   shapeType=Capsule
   length=10.2125
   radius=6.755
   shapeName=clavicle_l_capsule
   center=<8.061 -2.8442 -4.5358>
   rotation=<rot 25.2365 -90 4.201>
  }]
 clavicle_r [{
   shapeType=Capsule
   length=12.5594
   radius=7.6748
   shapeName=clavicle_r_capsule
   center=<-8.0717 1.1521 3.0739>
   rotation=<rot 25.9197 -90 9.0417>
  }]
 lowerarm_l []
 lowerarm_r []
 pelvis [{shapeType=Capsule, length=0.6528, radius=19.4095, center=<-1.9544 -9.0973 -0.1981>}]
 spine_01 [{shapeType=Capsule, length=10.0171, radius=13.7354, center=<-2.8183 -1.0873 0>}]
 spine_02 [{shapeType=Capsule, length=0, radius=17, shapeName=spine_02_capsule, center=<0 1.6712 0>}]
 spine_03 [
  {shapeType=Capsule, length=5, radius=18.4854, center=<-0.0747 2.3725 0>}
  {shapeType=Capsule, length=20.7188, radius=9.2812, center=<-1.2326 2.6059 0>, rotation=<rot 90 0 0>}
 ]
 spine_04 [
  {shapeType=Capsule, length=19.0796, radius=7.3916, center=<6.9796 -1.0098 0>}
  {shapeType=Capsule, length=10.0312, radius=9.2812, center=<-1.2326 1.308 0>, rotation=<rot 90 0 0>}
 ]
 spine_05 [{shapeType=Capsule, length=11.7359, radius=9.3291, center=<1.5198 -0.8941 0>}]
 upperarm_l []
 upperarm_r []
}
constraints [
 {
  bone1Name=spine_03
  bone2Name=spine_02
  swing1LimitDegrees=45
  swing1Motion=Free
  swing2LimitDegrees=45
  swing2Motion=Free
  twistLimitDegrees=45
  twistMotion=Free
 }
 {
  bone1Name=spine_05
  bone2Name=spine_04
  swing1LimitDegrees=45
  swing1Motion=Free
  swing2LimitDegrees=45
  swing2Motion=Free
  twistLimitDegrees=45
  twistMotion=Free
 }
 {
  bone1Name=spine_04
  bone2Name=spine_03
  swing1LimitDegrees=45
  swing1Motion=Free
  swing2LimitDegrees=45
  swing2Motion=Free
  twistLimitDegrees=45
  twistMotion=Free
 }
 {
  bone1Name=cape_root
  bone2Name=spine_05
  swing1LimitDegrees=45
  swing1Motion=Free
  swing2LimitDegrees=45
  swing2Motion=Free
  twistLimitDegrees=45
  twistMotion=Free
 }
 {
  bone1Name=cape_mid_01_l
  bone2Name=cape_root
  swing1LimitDegrees=45
  swing1Motion=Free
  swing2LimitDegrees=45
  swing2Motion=Free
  twistLimitDegrees=45
  twistMotion=Free
 }
 {bone1Name=cape_mid_02_l, bone2Name=cape_mid_01_l, swing1LimitDegrees=45, swing2LimitDegrees=5, twistLimitDegrees=5}
 {bone1Name=cape_mid_03_l, bone2Name=cape_mid_02_l, swing1LimitDegrees=45, swing2LimitDegrees=5, twistLimitDegrees=5}
 {bone1Name=cape_mid_04_l, bone2Name=cape_mid_03_l, swing1LimitDegrees=45, swing2LimitDegrees=45, twistLimitDegrees=5}
 {
  bone1Name=cape_mid_01_r
  bone2Name=cape_root
  swing1LimitDegrees=45
  swing1Motion=Free
  swing2LimitDegrees=45
  swing2Motion=Free
  twistLimitDegrees=45
  twistMotion=Free
 }
 {bone1Name=cape_mid_02_r, bone2Name=cape_mid_01_r, swing1LimitDegrees=45, swing2LimitDegrees=5, twistLimitDegrees=5}
 {bone1Name=cape_mid_03_r, bone2Name=cape_mid_02_r, swing1LimitDegrees=45, swing2LimitDegrees=5, twistLimitDegrees=5}
 {bone1Name=cape_mid_04_r, bone2Name=cape_mid_03_r, swing1LimitDegrees=45, swing2LimitDegrees=45, twistLimitDegrees=5}
 {
  bone1Name=clavicle_l
  bone2Name=spine_05
  swing1LimitDegrees=45
  swing1Motion=Free
  swing2LimitDegrees=45
  swing2Motion=Free
  twistLimitDegrees=45
  twistMotion=Free
 }
 {
  bone1Name=clavicle_cape_l
  bone2Name=clavicle_l
  swing1LimitDegrees=45
  swing1Motion=Free
  swing2LimitDegrees=45
  swing2Motion=Free
  twistLimitDegrees=45
  twistMotion=Free
 }
 {bone1Name=cape_01_l, bone2Name=clavicle_cape_l, swing1LimitDegrees=5, swing2LimitDegrees=5, twistLimitDegrees=15}
 {bone1Name=cape_02_l, bone2Name=cape_01_l, swing1LimitDegrees=45, swing2LimitDegrees=15, twistLimitDegrees=5}
 {bone1Name=cape_03_l, bone2Name=cape_02_l, swing1LimitDegrees=45, swing2LimitDegrees=15, twistLimitDegrees=5}
 {bone1Name=cape_04_l, bone2Name=cape_03_l, swing1LimitDegrees=45, swing2LimitDegrees=15, twistLimitDegrees=5}
 {
  bone1Name=clavicle_r
  bone2Name=spine_05
  swing1LimitDegrees=45
  swing1Motion=Free
  swing2LimitDegrees=45
  swing2Motion=Free
  twistLimitDegrees=45
  twistMotion=Free
 }
 {
  bone1Name=clavicle_cape_r
  bone2Name=clavicle_r
  swing1LimitDegrees=45
  swing1Motion=Free
  swing2LimitDegrees=45
  swing2Motion=Free
  twistLimitDegrees=45
  twistMotion=Free
 }
 {bone1Name=cape_01_r, bone2Name=clavicle_cape_r, swing1LimitDegrees=5, swing2LimitDegrees=5, twistLimitDegrees=15}
 {bone1Name=cape_02_r, bone2Name=cape_01_r, swing1LimitDegrees=45, swing2LimitDegrees=15, twistLimitDegrees=5}
 {bone1Name=cape_03_r, bone2Name=cape_02_r, swing1LimitDegrees=45, swing2LimitDegrees=15, twistLimitDegrees=5}
 {bone1Name=cape_04_r, bone2Name=cape_03_r, swing1LimitDegrees=45, swing2LimitDegrees=15, twistLimitDegrees=5}
 {
  bone1Name=spine_02
  bone2Name=spine_01
  swing1LimitDegrees=45
  swing1Motion=Free
  swing2LimitDegrees=45
  swing2Motion=Free
  twistLimitDegrees=45
  twistMotion=Free
 }
 {
  bone1Name=spine_01
  bone2Name=pelvis
  swing1LimitDegrees=45
  swing1Motion=Free
  swing2LimitDegrees=45
  swing2Motion=Free
  twistLimitDegrees=45
  twistMotion=Free
 }
 {
  bone1Name=upperarm_l
  bone2Name=clavicle_l
  swing1LimitDegrees=45
  swing1Motion=Free
  swing2LimitDegrees=45
  swing2Motion=Free
  twistLimitDegrees=45
  twistMotion=Free
 }
 {
  bone1Name=lowerarm_l
  bone2Name=upperarm_l
  swing1LimitDegrees=45
  swing1Motion=Free
  swing2LimitDegrees=45
  swing2Motion=Free
  twistLimitDegrees=45
  twistMotion=Free
 }
 {
  bone1Name=upperarm_r
  bone2Name=clavicle_r
  swing1LimitDegrees=45
  swing1Motion=Free
  swing2LimitDegrees=45
  swing2Motion=Free
  twistLimitDegrees=45
  twistMotion=Free
 }
 {
  bone1Name=lowerarm_r
  bone2Name=upperarm_r
  swing1LimitDegrees=45
  swing1Motion=Free
  swing2LimitDegrees=45
  swing2Motion=Free
  twistLimitDegrees=45
  twistMotion=Free
 }
 {
  bone1Name=cape_04_r
  bone2Name=cape_mid_04_r
  swing1LimitDegrees=45
  swing1Motion=Free
  swing2LimitDegrees=45
  swing2Motion=Free
  twistLimitDegrees=45
  twistMotion=Free
 }
 {
  bone1Name=cape_04_l
  bone2Name=cape_mid_04_l
  swing1LimitDegrees=45
  swing1Motion=Free
  swing2LimitDegrees=45
  swing2Motion=Free
  twistLimitDegrees=45
  twistMotion=Free
 }
]
