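# Physics asset for the jacket of the F_MED_SplitDiamond female player body, defining capsule bodies and joints for its dynamic bones.
# Seven chains (a–g) each run from a tiny base body on spine_03 through two jacket segments.
# Segment joints have tight, uneven swing and twist limits, while joints to spine_03 and between the spine bodies rotate freely.
# Larger capsules represent spine_02, spine_03 and only the right upper arm.
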PhysicsAsset /BRCosmetics/Characters/Player/Female/Medium/Bodies/F_MED_SplitDiamond/Meshes/F_MED_SplitDiamond_Jacket_Physics
bodies {
 dyn_jacket_01_a [{shapeType=Capsule, length=4.75, radius=1.5, shapeName=dyn_jacket_01_a_capsule, center=<2 0 0>, rotation=<rot 0 0 90>}]
 dyn_jacket_01_b [{shapeType=Capsule, length=6.75, radius=1.25, shapeName=dyn_jacket_01_b_capsule, center=<2 0 0>, rotation=<rot 0 0 90>}]
 dyn_jacket_01_c [{shapeType=Capsule, length=3.5, radius=1.5, shapeName=dyn_jacket_01_c_capsule, center=<2 0 0>, rotation=<rot 0 0 90>}]
 dyn_jacket_01_d [{shapeType=Capsule, length=3.5, radius=1.5, shapeName=dyn_jacket_01_d_capsule, center=<2 0 0>, rotation=<rot 0 0 90>}]
 dyn_jacket_01_e [{shapeType=Capsule, length=3.5, radius=1.5, shapeName=dyn_jacket_01_e_capsule, center=<2 0 0>, rotation=<rot 0 0 90>}]
 dyn_jacket_01_f [{shapeType=Capsule, length=3.5, radius=1.5, shapeName=dyn_jacket_01_f_capsule, center=<2 0 0>, rotation=<rot 0 0 90>}]
 dyn_jacket_01_g [{shapeType=Capsule, length=3.5, radius=1.5, shapeName=dyn_jacket_01_g_capsule, center=<2 0 0>, rotation=<rot 0 0 90>}]
 dyn_jacket_02_a [{shapeType=Capsule, length=5.5, radius=1.5, shapeName=dyn_jacket_02_a_capsule, center=<2 0 0>, rotation=<rot 0 0 90>}]
 dyn_jacket_02_b [{shapeType=Capsule, length=6.75, radius=1.25, shapeName=dyn_jacket_02_b_capsule, center=<2 0 0>, rotation=<rot 0 0 90>}]
 dyn_jacket_02_c [{shapeType=Capsule, length=3.5, radius=1.5, shapeName=dyn_jacket_02_c_capsule, center=<2 0 0>, rotation=<rot 0 0 90>}]
 dyn_jacket_02_d [{shapeType=Capsule, length=3.5, radius=1.5, shapeName=dyn_jacket_02_d_capsule, center=<2 0 0>, rotation=<rot 0 0 90>}]
 dyn_jacket_02_e [{shapeType=Capsule, length=3.5, radius=1.5, shapeName=dyn_jacket_02_e_capsule, center=<2 0 0>, rotation=<rot 0 0 90>}]
 dyn_jacket_02_f [{shapeType=Capsule, length=3.5, radius=1.5, shapeName=dyn_jacket_02_f_capsule, center=<2 0 0>, rotation=<rot 0 0 90>}]
 dyn_jacket_02_g [{shapeType=Capsule, length=3.5, radius=1.5, shapeName=dyn_jacket_02_g_capsule, center=<2 0 0>, rotation=<rot 0 0 90>}]
 dyn_jacket_base_a [{shapeType=Capsule, length=0.505, radius=0.505, shapeName=dyn_jacket_base_a_capsule}]
 dyn_jacket_base_b [{shapeType=Capsule, length=0.505, radius=0.505, shapeName=dyn_jacket_base_b_capsule}]
 dyn_jacket_base_c [{shapeType=Capsule, length=0.505, radius=0.505, shapeName=dyn_jacket_base_c_capsule}]
 dyn_jacket_base_d [{shapeType=Capsule, length=0.505, radius=0.505, shapeName=dyn_jacket_base_d_capsule}]
 dyn_jacket_base_e [{shapeType=Capsule, length=0.505, radius=0.505, shapeName=dyn_jacket_base_e_capsule}]
 dyn_jacket_base_f [{shapeType=Capsule, length=0.505, radius=0.505, shapeName=dyn_jacket_base_f_capsule}]
 dyn_jacket_base_g [{shapeType=Capsule, length=0.505, radius=0.505, shapeName=dyn_jacket_base_g_capsule}]
 spine_02 [{shapeType=Capsule, length=4.75, radius=8, shapeName=spine_02_capsule, center=<0 1.6712 0>}]
 spine_03 [{shapeType=Capsule, length=4.75, radius=8, shapeName=spine_03_capsule, center=<0 1.6712 0>}]
 upperarm_r [{shapeType=Capsule, length=15.0222, radius=4.1864, shapeName=upperarm_r_capsule, center=<-15.7764 0 0>, rotation=<rot 0 90 90>}]
}
constraints [
 {bone1Name=dyn_jacket_01_a, bone2Name=dyn_jacket_base_a, swing1LimitDegrees=3, swing2LimitDegrees=6, twistLimitDegrees=2}
 {bone1Name=dyn_jacket_02_a, bone2Name=dyn_jacket_01_a, swing1LimitDegrees=3, swing2LimitDegrees=6, twistLimitDegrees=2}
 {bone1Name=dyn_jacket_01_b, bone2Name=dyn_jacket_base_b, swing1LimitDegrees=3, swing2LimitDegrees=6, twistLimitDegrees=2}
 {bone1Name=dyn_jacket_02_b, bone2Name=dyn_jacket_01_b, swing1LimitDegrees=3, swing2LimitDegrees=6, twistLimitDegrees=2}
 {bone1Name=dyn_jacket_01_c, bone2Name=dyn_jacket_base_c, swing1LimitDegrees=3, swing2LimitDegrees=6, twistLimitDegrees=2}
 {bone1Name=dyn_jacket_02_c, bone2Name=dyn_jacket_01_c, swing1LimitDegrees=3, swing2LimitDegrees=6, twistLimitDegrees=2}
 {bone1Name=dyn_jacket_01_d, bone2Name=dyn_jacket_base_d, swing1LimitDegrees=3, swing2LimitDegrees=6, twistLimitDegrees=2}
 {bone1Name=dyn_jacket_02_d, bone2Name=dyn_jacket_01_d, swing1LimitDegrees=3, swing2LimitDegrees=6, twistLimitDegrees=2}
 {bone1Name=dyn_jacket_01_e, bone2Name=dyn_jacket_base_e, swing1LimitDegrees=3, swing2LimitDegrees=6, twistLimitDegrees=2}
 {bone1Name=dyn_jacket_02_e, bone2Name=dyn_jacket_01_e, swing1LimitDegrees=3, swing2LimitDegrees=6, twistLimitDegrees=2}
 {bone1Name=dyn_jacket_01_f, bone2Name=dyn_jacket_base_f, swing1LimitDegrees=3, swing2LimitDegrees=6, twistLimitDegrees=2}
 {bone1Name=dyn_jacket_02_f, bone2Name=dyn_jacket_01_f, swing1LimitDegrees=3, swing2LimitDegrees=6, twistLimitDegrees=2}
 {bone1Name=dyn_jacket_01_g, bone2Name=dyn_jacket_base_g, swing1LimitDegrees=3, swing2LimitDegrees=6, twistLimitDegrees=2}
 {bone1Name=dyn_jacket_02_g, bone2Name=dyn_jacket_01_g, swing1LimitDegrees=3, swing2LimitDegrees=6, twistLimitDegrees=2}
 {
  bone1Name=dyn_jacket_base_a
  bone2Name=spine_03
  swing1LimitDegrees=45
  swing1Motion=Free
  swing2LimitDegrees=45
  swing2Motion=Free
  twistLimitDegrees=45
  twistMotion=Free
 }
 {
  bone1Name=dyn_jacket_base_b
  bone2Name=spine_03
  swing1LimitDegrees=45
  swing1Motion=Free
  swing2LimitDegrees=45
  swing2Motion=Free
  twistLimitDegrees=45
  twistMotion=Free
 }
 {
  bone1Name=dyn_jacket_base_c
  bone2Name=spine_03
  swing1LimitDegrees=45
  swing1Motion=Free
  swing2LimitDegrees=45
  swing2Motion=Free
  twistLimitDegrees=45
  twistMotion=Free
 }
 {
  bone1Name=dyn_jacket_base_d
  bone2Name=spine_03
  swing1LimitDegrees=45
  swing1Motion=Free
  swing2LimitDegrees=45
  swing2Motion=Free
  twistLimitDegrees=45
  twistMotion=Free
 }
 {
  bone1Name=dyn_jacket_base_e
  bone2Name=spine_03
  swing1LimitDegrees=45
  swing1Motion=Free
  swing2LimitDegrees=45
  swing2Motion=Free
  twistLimitDegrees=45
  twistMotion=Free
 }
 {
  bone1Name=dyn_jacket_base_f
  bone2Name=spine_03
  swing1LimitDegrees=45
  swing1Motion=Free
  swing2LimitDegrees=45
  swing2Motion=Free
  twistLimitDegrees=45
  twistMotion=Free
 }
 {
  bone1Name=dyn_jacket_base_g
  bone2Name=spine_03
  swing1LimitDegrees=45
  swing1Motion=Free
  swing2LimitDegrees=45
  swing2Motion=Free
  twistLimitDegrees=45
  twistMotion=Free
 }
 {
  bone1Name=spine_03
  bone2Name=spine_02
  swing1LimitDegrees=45
  swing1Motion=Free
  swing2LimitDegrees=45
  swing2Motion=Free
  twistLimitDegrees=45
  twistMotion=Free
 }
 {
  bone1Name=upperarm_r
  bone2Name=spine_03
  swing1LimitDegrees=45
  swing1Motion=Free
  swing2LimitDegrees=45
  swing2Motion=Free
  twistLimitDegrees=45
  twistMotion=Free
 }
]
