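# Physics asset for the muffler on the SplitDiamond female medium player body, defining collision bodies and joints for the hanging cloth piece.
# A five-segment capsule chain hangs from a small muffler_base body attached to spine_04, with tight swing and twist limits at each link.
# Larger capsules on spine_02 through spine_04 approximate the torso as collision volumes, with free-motion constraints between the spine bodies and from spine_04 to muffler_base.
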PhysicsAsset /BRCosmetics/Characters/Player/Female/Medium/Bodies/F_MED_SplitDiamond/Meshes/F_MED_SplitDiamond_Muffler_Physics
bodies {
 muffler_01 [{shapeType=Capsule, length=3.25, radius=1.75, shapeName=muffler_01_capsule, center=<1 1 0>}]
 muffler_02 [{shapeType=Capsule, length=3.25, radius=1.75, shapeName=muffler_02_capsule, center=<1 1 0>}]
 muffler_03 [{shapeType=Capsule, length=3.25, radius=1.75, shapeName=muffler_03_capsule, center=<1 1 0>}]
 muffler_04 [{shapeType=Capsule, length=3.25, radius=1.75, shapeName=muffler_04_capsule, center=<1 1 0>}]
 muffler_05 [{shapeType=Capsule, length=3.25, radius=1.75, shapeName=muffler_05_capsule, center=<1 1 0>}]
 muffler_base [{shapeType=Capsule, length=0.505, radius=0.505, shapeName=muffler_base_capsule, rotation=<rot 0 -90 0>}]
 spine_02 [
  {shapeType=Capsule, length=3.2984, radius=8.4854, center=<1.179 2.2974 0>}
  {shapeType=Capsule, length=20.6875, radius=9.3125, center=<-1.8317 1.7235 0>, rotation=<rot 90 0 0>}
 ]
 spine_03 [{shapeType=Capsule, length=20.9688, radius=9.0312, center=<-0.0942 1.8854 0>, rotation=<rot 90 0 0>}]
 spine_04 [
  {shapeType=Capsule, length=7.8296, radius=8.3916, center=<5.5182 1.4112 0>}
  {shapeType=Capsule, length=10.0312, radius=9.2812, center=<-1.2326 1.308 0>, rotation=<rot 90 0 0>}
 ]
}
constraints [
 {
  bone1Name=spine_03
  bone2Name=spine_02
  swing1LimitDegrees=45
  swing1Motion=Free
  swing2LimitDegrees=45
  swing2Motion=Free
  twistLimitDegrees=45
  twistMotion=Free
 }
 {bone1Name=muffler_05, bone2Name=muffler_04, swing1LimitDegrees=15, swing2LimitDegrees=15, twistLimitDegrees=5}
 {
  bone1Name=muffler_base
  bone2Name=spine_04
  swing1LimitDegrees=45
  swing1Motion=Free
  swing2LimitDegrees=45
  swing2Motion=Free
  twistLimitDegrees=45
  twistMotion=Free
 }
 {bone1Name=muffler_01, bone2Name=muffler_base, swing1LimitDegrees=10, swing2LimitDegrees=10, twistLimitDegrees=5}
 {bone1Name=muffler_02, bone2Name=muffler_01, swing1LimitDegrees=15, swing2LimitDegrees=10, twistLimitDegrees=5}
 {bone1Name=muffler_03, bone2Name=muffler_02, swing1LimitDegrees=15, swing2LimitDegrees=3, twistLimitDegrees=5}
 {bone1Name=muffler_04, bone2Name=muffler_03, swing1LimitDegrees=15, swing2LimitDegrees=15, twistLimitDegrees=5}
 {
  bone1Name=spine_04
  bone2Name=spine_03
  swing1LimitDegrees=45
  swing1Motion=Free
  swing2LimitDegrees=45
  swing2Motion=Free
  twistLimitDegrees=45
  twistMotion=Free
 }
]
