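# Animation Blueprint for the F_MED_SplitDiamond cosmetic body, adding secondary motion to its cape, scarf, jacket, hood and boot straps.
# The parent mesh's pose runs through a deform Control Rig and corrective bone offsets into switched cape RigidBody setups, including skydive and backpack variants.
# Muffler and jacket RigidBody plus strap and hood AnimDynamics follow.
# EventGraph derives bone-delta correctives, state flags and a front-end curve weight blending DefaultSlot from scarf_front.
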AnimBlueprint /BRCosmetics/Characters/Player/Female/Medium/Bodies/F_MED_SplitDiamond/Meshes/F_MED_SplitDiamond_AnimBP
target_skeleton=&/F_MED_SplitDiamond_Skeleton
graphs [
 AnimGraph {nodes [
   Output Pose(Result=Layered blend per bone(BasePose=Use cached pose 'Simulation'(), BlendPoses_0=Slot 'DefaultSlot'(), BlendWeights_0=Get Custom Idle() -> (Custom Idle:double)) {layerSetup [{branchFilters [{boneName=scarf_front}]}]}) {layerGroup=None}
   n1 = Control Rig(Source=Copy Pose From Mesh() {bUseAttachedParent=true}) {
    alphaBoolBlend {blendOption=Linear}
    controlRigAssetReference {blueprintRigClass=&/F_MED_Split_Diamond_DeformRig.F_MED_Split_Diamond_DeformRig_C}
   }
   n2 = "Transform (Modify) Bone - Bone: scarf_front"(ComponentPose=Local To Component(LocalPose=@n1), Alpha=1, Translation="0,0,0", Rotation="0,0,0", Scale="1,1,1") {rotationMode=BMM_Replace, rotationSpace=BCS_BoneSpace, alphaBoolBlend {blendOption=Linear}, boneToModify {boneName=scarf_front}}
   n3 = "Transform (Modify) Bone - Bone: scarf_back"(ComponentPose=@n2, Alpha=1, Translation="0,0,0", Rotation="0,-60,0", Scale="1,1,1") {
    rotationMode=BMM_Additive
    rotationSpace=BCS_BoneSpace
    alphaBoolBlend {blendOption=Linear}
    boneToModify {boneName=scarf_back}
    rotation {yaw=-60}
   }
   n4 = "Transform (Modify) Bone - Bone: hood_clavicle_base"(
    ComponentPose=@n3
    Alpha=1
    Translation="0,0,0"
    Rotation=Get clavicle_r_hood_base_rot() -> (clavicle_r_hood_base_rot:Rotator)
    Scale="1,1,1"
   ) {
    rotationMode=BMM_Additive
    rotationSpace=BCS_BoneSpace
    translationMode=BMM_Additive
    translationSpace=BCS_BoneSpace
    alphaBoolBlend {blendOption=Linear}
    boneToModify {boneName=hood_clavicle_base}
   }
   n5 = "Transform (Modify) Bone - Bone: hood_upperarm_front_r"(ComponentPose=@n4, Alpha=1, Translation="0,2,-12", Rotation="-20,0,0", Scale="1,1,1") {
    rotationMode=BMM_Additive
    rotationSpace=BCS_BoneSpace
    translationMode=BMM_Additive
    translationSpace=BCS_BoneSpace
    alphaBoolBlend {blendOption=Linear}
    boneToModify {boneName=hood_upperarm_front_r}
    rotation {pitch=-20}
    translation {y=2, z=-12}
   }
   n6 = "Transform (Modify) Bone - Bone: hood_upperarm_r"(ComponentPose=@n5, Alpha=1, Translation="2,-2,-2", Rotation="0,0,0", Scale="1,1,1") {
    rotationMode=BMM_Additive
    rotationSpace=BCS_BoneSpace
    translationMode=BMM_Additive
    translationSpace=BCS_BoneSpace
    alphaBoolBlend {blendOption=Linear}
    boneToModify {boneName=hood_upperarm_r}
    translation=<2 -2 -2>
   }
   n7 = "Transform (Modify) Bone - Bone: hood_upperarm_front_r"(
    ComponentPose=@n6
    Alpha=Get upper_arm_r_rot_down_alpha() -> (upper_arm_r_rot_down_alpha:double)
    Translation="0,-5,0"
    Rotation="0,0,0"
    Scale="1,1,1"
   ) {
    rotationMode=BMM_Additive
    rotationSpace=BCS_BoneSpace
    translationMode=BMM_Additive
    translationSpace=BCS_BoneSpace
    alphaBoolBlend {blendOption=Linear}
    boneToModify {boneName=hood_upperarm_front_r}
    translation {y=-5}
   }
   n8 = "Transform (Modify) Bone - Bone: hood_upperarm_front_02_r"(ComponentPose=@n7, Alpha=1, Translation="2,1.5,4", Rotation="0,0,0", Scale="1,1,1") {
    rotationMode=BMM_Additive
    rotationSpace=BCS_BoneSpace
    scaleMode=BMM_Replace
    scaleSpace=BCS_BoneSpace
    translationMode=BMM_Additive
    translationSpace=BCS_BoneSpace
    alphaBoolBlend {blendOption=Linear}
    boneToModify {boneName=hood_upperarm_front_02_r}
    translation=<2 1.5 4>
   }
   n9 = "Transform (Modify) Bone - Bone: hood_upperarm_r"(ComponentPose=@n8, Alpha=1, Translation="0,-1,-2", Rotation="0,0,0", Scale="1,1,1") {
    rotationMode=BMM_Additive
    rotationSpace=BCS_BoneSpace
    scaleMode=BMM_Replace
    scaleSpace=BCS_BoneSpace
    translationMode=BMM_Additive
    translationSpace=BCS_BoneSpace
    alphaBoolBlend {blendOption=Linear}
    boneToModify {boneName=hood_upperarm_r}
    translation {y=-1, z=-2}
   }
   n10 = "Transform (Modify) Bone - Bone: hood_upperarm_r"(ComponentPose=@n9, Alpha=1, Translation="2,0,0", Rotation="0,0,0", Scale="1,1,1") {
    rotationMode=BMM_Additive
    rotationSpace=BCS_BoneSpace
    scaleMode=BMM_Replace
    scaleSpace=BCS_BoneSpace
    translationMode=BMM_Additive
    translationSpace=BCS_BoneSpace
    alphaBoolBlend {blendOption=Linear}
    boneToModify {boneName=hood_upperarm_r}
    translation {x=2}
   }
   n11 = "Transform (Modify) Bone - Bone: collar_01_l"(ComponentPose=@n10, Alpha=1, Translation="0,1,5", Rotation="0,0,0", Scale="1,1,1") {
    rotationMode=BMM_Additive
    rotationSpace=BCS_BoneSpace
    scaleMode=BMM_Replace
    scaleSpace=BCS_BoneSpace
    translationMode=BMM_Additive
    translationSpace=BCS_BoneSpace
    alphaBoolBlend {blendOption=Linear}
    boneToModify {boneName=collar_01_l}
    translation {y=1, z=5}
   }
   n12 = "Transform (Modify) Bone - Bone: collar_01_r"(ComponentPose=@n11, Alpha=1, Translation="0,1,-1", Rotation="0,0,0", Scale="1,1,1") {
    rotationMode=BMM_Additive
    rotationSpace=BCS_BoneSpace
    scaleMode=BMM_Replace
    scaleSpace=BCS_BoneSpace
    translationMode=BMM_Additive
    translationSpace=BCS_BoneSpace
    alphaBoolBlend {blendOption=Linear}
    boneToModify {boneName=collar_01_r}
    translation {y=1, z=-1}
   }
   n13 = "Transform (Modify) Bone - Bone: attach_backpack"!(
    ComponentPose=@n12
    bAlphaBoolEnabled:bool=Get cape_backpack_equipped() -> (cape_backpack_equipped:bool)
    Translation="0,-10,0"
    Rotation="0,0,-2"
    Scale="1,1,0.75"
   ) -> (Pose) {
    alphaInputType=Bool
    rotationMode=BMM_Additive
    rotationSpace=BCS_BoneSpace
    translationMode=BMM_Additive
    translationSpace=BCS_BoneSpace
    alphaBoolBlend {blendOption=Linear}
    boneToModify {boneName=attach_backpack}
    rotation {roll=-2}
    scale {z=0.75}
    translation {y=-10}
   }
   Save cached pose 'Base Pose'(Pose=Component To Local(ComponentPose=@n13)) {cacheName=Base Pose}
   n14 = RigidBody!(ComponentPose=Local To Component(LocalPose=Use cached pose 'DeltasPose'()), Alpha, OverrideWorldGravity:Vector="0,0,-980", ExternalForce) -> (Pose) {
    bClampLinearTranslationLimitToRefPose=true
    bOverrideWorldGravity=true
    overridePhysicsAsset=&/F_MED_SplitDiamond_Cape_Emote_Physics
    simulationSpace=BaseBoneSpace
    alphaBoolBlend {blendOption=Linear}
    baseBoneRef {boneName=attach_cape}
    componentAppliedLinearAccClamp=<3000 3000 3000>
    componentLinearAccScale {z=0.6}
    componentLinearVelScale {z=0.6}
    overrideWorldGravity {z=-980}
    simSpaceSettings {
     maxAngularAcceleration=500
     maxAngularVelocity=3
     maxLinearAcceleration=700
     maxLinearVelocity=500
     velocityScaleZ=0
     worldAlpha=1
     externalLinearDragV=<1 1>
    }
   }
   n15 = RigidBody(ComponentPose=Local To Component(LocalPose=Use cached pose 'DeltasPose'()), Alpha=0.5, OverrideWorldGravity="0,0,-980") {
    bClampLinearTranslationLimitToRefPose=true
    bOverrideWorldGravity=true
    overridePhysicsAsset=&F_MED_SplitDiamond_Cape_Emote_Physics
    simulationSpace=BaseBoneSpace
    alphaBoolBlend {blendOption=Linear}
    alphaScaleBiasClamp {bInterpResult=true}
    baseBoneRef {boneName=attach_cape}
    componentAppliedLinearAccClamp=<3000 3000 3000>
    componentLinearAccScale {z=0.6}
    componentLinearVelScale {z=0.6}
    overrideWorldGravity {z=-980}
    simSpaceSettings {
     maxAngularAcceleration=500
     maxAngularVelocity=3
     maxLinearAcceleration=700
     maxLinearVelocity=500
     velocityScaleZ=0
     worldAlpha=1
     externalLinearDragV=<1 1>
    }
   }
   n16 = RigidBody!(
    ComponentPose=Local To Component(LocalPose=Use cached pose 'DeltasPose'())
    bAlphaBoolEnabled:bool=Get cape_backpack_equipped()
    OverrideWorldGravity="0,0,-980"
    ExternalForce
   ) -> (Pose) {
    alphaInputType=Bool
    bOverrideWorldGravity=true
    overridePhysicsAsset=&/F_MED_SplitDiamond_Cape_Physics
    simulationSpace=BaseBoneSpace
    alphaBoolBlend {blendOption=Linear}
    baseBoneRef {boneName=spine_05}
    componentAppliedLinearAccClamp=<2000 2000 2000>
    componentLinearAccScale=<0.2 0.2 0.3>
    componentLinearVelScale=<0.2 0.2 0.3>
    overrideWorldGravity {z=-980}
   }
   n17 = RigidBody(ComponentPose=@n16, bAlphaBoolEnabled=NOT Boolean(A=Get cape_backpack_equipped()), OverrideWorldGravity="0,0,-980") {
    alphaInputType=Bool
    bClampLinearTranslationLimitToRefPose=true
    bOverrideWorldGravity=true
    overridePhysicsAsset=&F_MED_SplitDiamond_Cape_Physics
    simulationSpace=BaseBoneSpace
    alphaBoolBlend {blendOption=Linear}
    baseBoneRef {boneName=attach_cape}
    componentAppliedLinearAccClamp=<3000 3000 3000>
    componentLinearAccScale {z=0.6}
    componentLinearVelScale {z=0.6}
    overrideWorldGravity {z=-980}
    simSpaceSettings {
     maxAngularAcceleration=500
     maxAngularVelocity=3
     maxLinearAcceleration=700
     maxLinearVelocity=500
     velocityScaleZ=0
     worldAlpha=1
     externalLinearDragV=<1 1>
    }
   }
   n18 = Blend Poses by bool(
    BlendPose_0=Blend Poses by bool(BlendPose_0=Component To Local(ComponentPose=@n15), BlendPose_1=Component To Local(ComponentPose=@n14)) {blendType=Linear, transitionType=Inertialization}
    BlendPose_1=Component To Local(ComponentPose=@n17)
   ) {blendType=Linear, transitionType=Inertialization}
   n19 = Use cached pose 'SkydiveSimPose'()
   n20 = RigidBody!(
    ComponentPose=Local To Component(LocalPose=Blend Poses by bool(BlendPose_0=@n19, BlendPose_1=@n18) {blendType=Linear, transitionType=Inertialization})
    Alpha:float
    ExternalForce
   ) -> (Pose) {
    bOverrideWorldGravity=true
    overridePhysicsAsset=&/F_MED_SplitDiamond_Muffler_Physics
    simulationSpace=BaseBoneSpace
    alphaBoolBlend {blendOption=Linear}
    baseBoneRef {boneName=spine_04}
    componentAppliedLinearAccClamp=<2000 2000 2000>
    componentLinearAccScale=<0.5 0.5 0.5>
    componentLinearVelScale=<0.5 0.5 0.5>
    overrideWorldGravity {z=-980}
   }
   n21 = RigidBody!(ComponentPose=@n20, bAlphaBoolEnabled:bool=Get cape_backpack_equipped(), ExternalForce) -> (Pose) {
    alphaInputType=Bool
    bOverrideWorldGravity=true
    overridePhysicsAsset=&/F_MED_SplitDiamond_Jacket_Cape_Off_Physics
    simulationSpace=BaseBoneSpace
    alphaBoolBlend {blendOption=Linear}
    baseBoneRef {boneName=spine_05}
    componentAppliedLinearAccClamp=<2000 2000 2000>
    componentLinearAccScale=<0.2 0.2 0.3>
    componentLinearVelScale=<0.2 0.2 0.3>
    overrideWorldGravity {z=-980}
   }
   n22 = RigidBody(ComponentPose=@n21, bAlphaBoolEnabled=NOT Boolean(A=Get cape_backpack_equipped())) {
    alphaInputType=Bool
    bOverrideWorldGravity=true
    overridePhysicsAsset=&/F_MED_SplitDiamond_Jacket_Physics
    simulationSpace=BaseBoneSpace
    alphaBoolBlend {blendOption=Linear}
    baseBoneRef {boneName=spine_05}
    componentAppliedLinearAccClamp=<2000 2000 2000>
    componentLinearAccScale=<0.35 0.35 0.4>
    componentLinearVelScale=<0.35 0.35 0.4>
    overrideWorldGravity {z=-980}
   }
   n23 = Anim Dynamics - Bone: boot_strap_01_a_l(ComponentPose=@n22, Alpha=1, ExternalForce="0,0,0") {
    angularDampingOverride=1
    angularSpringConstant=1600
    bAngularSpring=true
    bGravityOverrideInSimSpace=true
    bOverrideAngularDamping=true
    bOverrideLinearDamping=true
    bUseGravityOverride=true
    linearDampingOverride=0.85
    linearSpringConstant=10000
    simulationSpace=RootRelative
    alphaBoolBlend {blendOption=Linear}
    boundBone {boneName=boot_strap_01_a_l}
    componentAppliedLinearAccClamp=<6000 6000 6000>
    componentLinearAccScale=<1 1 1>
    componentLinearVelScale=<0.5 0.5 1>
    gravityOverride {z=-980}
    physicsBodyDefinitions [{
      boundBone {boneName=boot_strap_01_a_l}
      boxExtents=<5 5 5>
      constraintSetup {angularTargetAxis=AxisZ, angularLimitsMax=<15 5 15>, angularLimitsMin=<-15 -15 -15>, angularTarget=<0 0 0>}
      localJointOffset=<1 0 0>
     }]
    relativeSpaceBone {boneName=calf_l}
   }
   n24 = Anim Dynamics - Bone: boot_strap_02_a_l(ComponentPose=@n23, Alpha=1, ExternalForce="0,0,0") {
    angularDampingOverride=1
    angularSpringConstant=1600
    bAngularSpring=true
    bGravityOverrideInSimSpace=true
    bOverrideAngularDamping=true
    bOverrideLinearDamping=true
    bUseGravityOverride=true
    linearDampingOverride=0.85
    linearSpringConstant=10000
    simulationSpace=RootRelative
    alphaBoolBlend {blendOption=Linear}
    boundBone {boneName=boot_strap_02_a_l}
    componentAppliedLinearAccClamp=<6000 6000 6000>
    componentLinearAccScale=<1 1 1>
    componentLinearVelScale=<0.5 0.5 1>
    gravityOverride {z=-980}
    physicsBodyDefinitions [{
      boundBone {boneName=boot_strap_02_a_l}
      boxExtents=<5 5 5>
      constraintSetup {angularTargetAxis=AxisZ, angularLimitsMax=<5 10 60>, angularLimitsMin=<-5 -30 -60>, angularTarget=<0 0 0>}
      localJointOffset=<1.5 0 0>
     }]
    relativeSpaceBone {boneName=calf_l}
   }
   n25 = Anim Dynamics - Bone: boot_strap_01_a_r(ComponentPose=@n24, Alpha=1, ExternalForce="0,0,0") {
    angularDampingOverride=1
    angularSpringConstant=1600
    bAngularSpring=true
    bGravityOverrideInSimSpace=true
    bOverrideAngularDamping=true
    bOverrideLinearDamping=true
    bUseGravityOverride=true
    linearDampingOverride=0.85
    linearSpringConstant=10000
    simulationSpace=RootRelative
    alphaBoolBlend {blendOption=Linear}
    boundBone {boneName=boot_strap_01_a_r}
    componentAppliedLinearAccClamp=<6000 6000 6000>
    componentLinearAccScale=<0.8 0.8 0.8>
    componentLinearVelScale=<0.5 0.5 1>
    gravityOverride {z=-980}
    physicsBodyDefinitions [{
      boundBone {boneName=boot_strap_01_a_r}
      boxExtents=<6 5 5>
      constraintSetup {angularTargetAxis=AxisZ, angularLimitsMax=<15 5 30>, angularLimitsMin=<-15 -15 -45>, angularTarget=<0 0 0>}
      localJointOffset=<2 0 0>
     }]
    relativeSpaceBone {boneName=calf_r}
   }
   n26 = Anim Dynamics - Bone: boot_strap_02_a_r(ComponentPose=@n25, Alpha=1, ExternalForce="0,0,0") {
    angularDampingOverride=1
    angularSpringConstant=1600
    bAngularSpring=true
    bGravityOverrideInSimSpace=true
    bOverrideAngularDamping=true
    bOverrideLinearDamping=true
    bUseGravityOverride=true
    linearDampingOverride=0.85
    linearSpringConstant=10000
    simulationSpace=RootRelative
    alphaBoolBlend {blendOption=Linear}
    boundBone {boneName=boot_strap_02_a_r}
    componentAppliedLinearAccClamp=<4000 4000 4000>
    componentLinearAccScale=<0.7 0.7 0.7>
    componentLinearVelScale=<0.5 0.5 1>
    gravityOverride {z=-980}
    physicsBodyDefinitions [{
      boundBone {boneName=boot_strap_02_a_r}
      boxExtents=<6 5 5>
      constraintSetup {angularTargetAxis=AxisZ, angularLimitsMax=<5 10 30>, angularLimitsMin=<-5 -30 -30>, angularTarget=<0 0 0>}
      localJointOffset=<1 0 0>
     }]
    relativeSpaceBone {boneName=calf_r}
   }
   n27 = Anim Dynamics - Bone: boot_strap_01_b_l(ComponentPose=@n26, Alpha=1, ExternalForce="0,0,0") {
    angularDampingOverride=1
    angularSpringConstant=1600
    bAngularSpring=true
    bGravityOverrideInSimSpace=true
    bOverrideAngularDamping=true
    bOverrideLinearDamping=true
    bUseGravityOverride=true
    linearDampingOverride=0.85
    linearSpringConstant=10000
    simulationSpace=RootRelative
    alphaBoolBlend {blendOption=Linear}
    boundBone {boneName=boot_strap_01_b_l}
    componentAppliedLinearAccClamp=<6000 6000 6000>
    componentLinearAccScale=<1 1 1>
    componentLinearVelScale=<0.5 0.5 1>
    gravityOverride {z=-980}
    physicsBodyDefinitions [{
      boundBone {boneName=boot_strap_01_b_l}
      boxExtents=<3 4.5 3>
      constraintSetup {angularTargetAxis=AxisZ, angularLimitsMax=<15 15 15>, angularLimitsMin=<-15 -5 -15>, angularTarget=<0 0 0>}
      localJointOffset=<0.75 0 0>
     }]
    relativeSpaceBone {boneName=calf_l}
   }
   n28 = Anim Dynamics - Bone: boot_strap_02_b_l(ComponentPose=@n27, Alpha=1, ExternalForce="0,0,0") {
    angularDampingOverride=1
    angularSpringConstant=1600
    bAngularSpring=true
    bGravityOverrideInSimSpace=true
    bOverrideAngularDamping=true
    bOverrideLinearDamping=true
    bUseGravityOverride=true
    linearDampingOverride=0.85
    linearSpringConstant=10000
    simulationSpace=RootRelative
    alphaBoolBlend {blendOption=Linear}
    boundBone {boneName=boot_strap_02_b_l}
    componentAppliedLinearAccClamp=<6000 6000 6000>
    componentLinearAccScale=<1 1 1>
    componentLinearVelScale=<0.5 0.5 1>
    gravityOverride {z=-980}
    physicsBodyDefinitions [{
      boundBone {boneName=boot_strap_02_b_l}
      boxExtents=<3 4.5 3>
      constraintSetup {angularTargetAxis=AxisZ, angularLimitsMax=<5 30 60>, angularLimitsMin=<-5 -10 -60>, angularTarget=<0 0 0>}
      localJointOffset=<0.75 0 0>
     }]
    relativeSpaceBone {boneName=calf_l}
   }
   n29 = Anim Dynamics - Bone: boot_strap_01_b_r(ComponentPose=@n28, Alpha=1, ExternalForce="0,0,0") {
    angularDampingOverride=1
    angularSpringConstant=1600
    bAngularSpring=true
    bGravityOverrideInSimSpace=true
    bOverrideAngularDamping=true
    bOverrideLinearDamping=true
    bUseGravityOverride=true
    linearDampingOverride=0.85
    linearSpringConstant=10000
    simulationSpace=RootRelative
    alphaBoolBlend {blendOption=Linear}
    boundBone {boneName=boot_strap_01_b_r}
    componentAppliedLinearAccClamp=<4000 4000 4000>
    componentLinearAccScale=<0.8 0.8 0.8>
    componentLinearVelScale=<0.5 0.5 1>
    gravityOverride {z=-980}
    physicsBodyDefinitions [{
      boundBone {boneName=boot_strap_01_b_r}
      boxExtents=<4.5 2.5 2.5>
      constraintSetup {angularTargetAxis=AxisZ, angularLimitsMax=<15 15 30>, angularLimitsMin=<-15 -5 -45>, angularTarget=<0 0 0>}
      localJointOffset=<2 0 0>
     }]
    relativeSpaceBone {boneName=calf_r}
   }
   n30 = Anim Dynamics - Bone: boot_strap_02_b_r(ComponentPose=@n29, Alpha=1, ExternalForce="0,0,0") {
    angularDampingOverride=1
    angularSpringConstant=1600
    bAngularSpring=true
    bGravityOverrideInSimSpace=true
    bOverrideAngularDamping=true
    bOverrideLinearDamping=true
    bUseGravityOverride=true
    linearDampingOverride=0.85
    linearSpringConstant=10000
    simulationSpace=RootRelative
    alphaBoolBlend {blendOption=Linear}
    boundBone {boneName=boot_strap_02_b_r}
    componentAppliedLinearAccClamp=<4000 4000 4000>
    componentLinearAccScale=<0.8 0.8 0.8>
    componentLinearVelScale=<0.5 0.5 1>
    gravityOverride {z=-980}
    physicsBodyDefinitions [{
      boundBone {boneName=boot_strap_02_b_r}
      boxExtents=<4.5 2.5 2.5>
      constraintSetup {angularTargetAxis=AxisZ, angularLimitsMax=<5 30 30>, angularLimitsMin=<-5 10 -30>, angularTarget=<0 0 0>}
      localJointOffset=<1 0 0>
     }]
    relativeSpaceBone {boneName=calf_r}
   }
   n31 = Anim Dynamics - Bone: dyn_hood_01!(ComponentPose=@n30, Alpha=0.8, GravityOverride:Vector="0,0,-980", ExternalForce="0,0,0") -> (Pose) {
    angularSpringConstant=10000
    bAngularSpring=true
    bUseGravityOverride=true
    linearSpringConstant=0.1
    simulationSpace=BoneRelative
    alphaBoolBlend {blendOption=Linear}
    boundBone {boneName=dyn_hood_01}
    componentAppliedLinearAccClamp=<2000 2000 2000>
    componentLinearAccScale=<0.5 0.5 0.6>
    componentLinearVelScale=<0.5 0.5 0.6>
    gravityOverride {z=-980}
    physicsBodyDefinitions [{
      boundBone {boneName=dyn_hood_01}
      boxExtents=<15 8 15>
      constraintSetup {angularLimitsMax=<10 2 3>, angularLimitsMin=<-10 -2 -20>, angularTarget=<0 0 0>}
      localJointOffset=<1 0 0>
     }]
    relativeSpaceBone {boneName=attach_cape}
   }
   n32 = Anim Dynamics - Bone: dyn_hood_02!(ComponentPose=@n31, Alpha=1, GravityOverride:Vector="0,0,-980", ExternalForce="0,0,0") -> (Pose) {
    angularSpringConstant=10000
    bAngularSpring=true
    bUseGravityOverride=true
    linearSpringConstant=0.1
    simulationSpace=BoneRelative
    alphaBoolBlend {blendOption=Linear}
    boundBone {boneName=dyn_hood_02}
    componentAppliedLinearAccClamp=<2000 2000 2000>
    componentLinearAccScale=<0.5 0.5 0.6>
    componentLinearVelScale=<0.5 0.5 0.6>
    gravityOverride {z=-980}
    physicsBodyDefinitions [{
      boundBone {boneName=dyn_hood_02}
      boxExtents=<15 8 15>
      constraintSetup {angularLimitsMax=<10 2 3>, angularLimitsMin=<-10 -2 -30>, angularTarget=<0 0 0>}
      localJointOffset=<1 0 0>
     }]
    relativeSpaceBone {boneName=attach_cape}
   }
   n33 = Anim Dynamics - Bone: dyn_hood_03!(ComponentPose=@n32, Alpha=1, GravityOverride:Vector="0,0,-980", ExternalForce="0,0,0") -> (Pose) {
    angularSpringConstant=10000
    bAngularSpring=true
    bUseGravityOverride=true
    linearSpringConstant=0.1
    simulationSpace=BoneRelative
    alphaBoolBlend {blendOption=Linear}
    boundBone {boneName=dyn_hood_03}
    componentAppliedLinearAccClamp=<2000 2000 2000>
    componentLinearAccScale=<0.5 0.5 0.6>
    componentLinearVelScale=<0.5 0.5 0.6>
    gravityOverride {z=-980}
    physicsBodyDefinitions [{
      boundBone {boneName=dyn_hood_03}
      boxExtents=<10 8 15>
      constraintSetup {angularLimitsMax=<10 2 3>, angularLimitsMin=<-10 -2 -30>, angularTarget=<0 0 0>}
      localJointOffset=<1 0 0>
     }]
    relativeSpaceBone {boneName=attach_cape}
   }
   Save cached pose 'Simulation'(Pose=Component To Local(ComponentPose=@n33)) {cacheName=Simulation}
   n34 = Use cached pose 'Simulation'()
   Save cached pose 'PostSim'(Pose=Component To Local(ComponentPose=Local To Component(LocalPose=@n34))) {cacheName=PostSim}
   n35 = "Transform (Modify) Bone - Bone: clavicle_cape_r"(ComponentPose=Local To Component(LocalPose=Use cached pose 'Base Pose'()), Alpha=1, Translation="0,0,0", Rotation="10,-20,0", Scale="1,1,1") {
    rotationMode=BMM_Additive
    rotationSpace=BCS_BoneSpace
    alphaBoolBlend {blendOption=Linear}
    boneToModify {boneName=clavicle_cape_r}
    rotation {pitch=10, yaw=-20}
   }
   n36 = "Transform (Modify) Bone - Bone: clavicle_cape_l"(ComponentPose=@n35, Alpha=1, Translation="0,0,0", Rotation="10,-10,0", Scale="1,1,1") {
    rotationMode=BMM_Additive
    rotationSpace=BCS_BoneSpace
    alphaBoolBlend {blendOption=Linear}
    boneToModify {boneName=clavicle_cape_l}
    rotation {pitch=10, yaw=-10}
   }
   n37 = "Transform (Modify) Bone - Bone: cape_01_l"(ComponentPose=@n36, Alpha=1, Translation="0,0,0", Rotation="-35,-15,0", Scale="1,1,1") {
    rotationMode=BMM_Additive
    rotationSpace=BCS_BoneSpace
    alphaBoolBlend {blendOption=Linear}
    boneToModify {boneName=cape_01_l}
    rotation {pitch=-35, yaw=-15}
   }
   n38 = "Transform (Modify) Bone - Bone: cape_01_r"(ComponentPose=@n37, Alpha=1, Translation="0,0,0", Rotation="30,-15,0", Scale="1,1,1") {
    rotationMode=BMM_Additive
    rotationSpace=BCS_BoneSpace
    alphaBoolBlend {blendOption=Linear}
    boneToModify {boneName=cape_01_r}
    rotation {pitch=30, yaw=-15}
   }
   n39 = "Transform (Modify) Bone - Bone: collar_01_l"(ComponentPose=@n38, Alpha=1, Translation="0,0,2", Rotation="0,0,15", Scale="1,1,1") {
    rotationMode=BMM_Additive
    rotationSpace=BCS_BoneSpace
    translationMode=BMM_Additive
    translationSpace=BCS_BoneSpace
    alphaBoolBlend {blendOption=Linear}
    boneToModify {boneName=collar_01_l}
    rotation {roll=15}
    translation {z=2}
   }
   n40 = "Transform (Modify) Bone - Bone: clavicle_cape_r"!(ComponentPose=@n39, bAlphaBoolEnabled:bool=True, Translation="0,0,0", Rotation="0,-10,5", Scale="1,1,1") -> (Pose) {
    alphaInputType=Bool
    rotationMode=BMM_Additive
    rotationSpace=BCS_BoneSpace
    alphaBoolBlend {blendInTime=0.1, blendOption=Linear, blendOutTime=0.1}
    boneToModify {boneName=clavicle_cape_r}
    rotation {roll=5, yaw=-10}
   }
   n41 = "Transform (Modify) Bone - Bone: clavicle_cape_l"!(ComponentPose=@n40, bAlphaBoolEnabled:bool=True, Translation="0,0,0", Rotation="0,-10,5", Scale="1,1,1") -> (Pose) {
    alphaInputType=Bool
    rotationMode=BMM_Additive
    rotationSpace=BCS_BoneSpace
    alphaBoolBlend {blendInTime=0.1, blendOption=Linear, blendOutTime=0.1}
    boneToModify {boneName=clavicle_cape_l}
    rotation {roll=5, yaw=-10}
   }
   Save cached pose 'DeltasPose'(Pose=Component To Local(ComponentPose=@n41)) {cacheName=DeltasPose}
   n42 = NOT Boolean(A=Get bIsParachuteOpened())
   n43 = "Transform (Modify) Bone - Bone: clavicle_cape_r"(
    ComponentPose=Local To Component(LocalPose=Use cached pose 'Base Pose'())
    bAlphaBoolEnabled=@n42
    Translation="0,0,0"
    Rotation="30,0,0"
    Scale="1,1,1"
   ) {
    alphaInputType=Bool
    rotationMode=BMM_Additive
    rotationSpace=BCS_BoneSpace
    alphaBoolBlend {blendOption=Linear}
    boneToModify {boneName=clavicle_cape_r}
    rotation {pitch=30}
   }
   n44 = "Transform (Modify) Bone - Bone: clavicle_cape_l"(ComponentPose=@n43, bAlphaBoolEnabled=@n42, Translation="0,0,0", Rotation="30,0,0", Scale="1,1,1") {
    alphaInputType=Bool
    rotationMode=BMM_Additive
    rotationSpace=BCS_BoneSpace
    alphaBoolBlend {blendOption=Linear}
    boneToModify {boneName=clavicle_cape_l}
    rotation {pitch=30}
   }
   n45 = "Transform (Modify) Bone - Bone: cape_01_r"!(ComponentPose=@n44, bAlphaBoolEnabled:bool=@n42, Translation="0,0,0", Rotation="0,0,-30", Scale="1,1,1") -> (Pose) {
    alphaInputType=Bool
    rotationMode=BMM_Additive
    rotationSpace=BCS_BoneSpace
    alphaBoolBlend {blendOption=Linear}
    boneToModify {boneName=cape_01_r}
    rotation {roll=-30}
   }
   n46 = "Transform (Modify) Bone - Bone: cape_01_l"!(ComponentPose=@n45, bAlphaBoolEnabled:bool=@n42, Translation="0,0,0", Rotation="0,20,60", Scale="1,1,1") -> (Pose) {
    alphaInputType=Bool
    rotationMode=BMM_Additive
    rotationSpace=BCS_BoneSpace
    alphaBoolBlend {blendOption=Linear}
    boneToModify {boneName=cape_01_l}
    rotation {roll=60, yaw=20}
   }
   n47 = RigidBody!(ComponentPose=@n46, bAlphaBoolEnabled=Get cape_backpack_equipped(), OverrideWorldGravity:Vector="0,0,0", ExternalForce) -> (Pose) {
    alphaInputType=Bool
    bOverrideWorldGravity=true
    overridePhysicsAsset=&/F_MED_SplitDiamond_Cape_Backpack_Physics
    simulationSpace=BaseBoneSpace
    alphaBoolBlend {blendOption=Linear}
    baseBoneRef {boneName=spine_05}
    componentAppliedLinearAccClamp=<2000 2000 2000>
    componentLinearAccScale=<0.1 0.1 0.1>
    componentLinearVelScale=<0.1 0.1 0.1>
    overrideWorldGravity {z=-980}
   }
   n48 = RigidBody(ComponentPose=@n47, bAlphaBoolEnabled=NOT Boolean(A=Get cape_backpack_equipped()), OverrideWorldGravity="0,0,-980") {
    alphaInputType=Bool
    bOverrideWorldGravity=true
    overridePhysicsAsset=&/F_MED_SplitDiamond_Cape_NyxGlass_Physics
    simulationSpace=BaseBoneSpace
    alphaBoolBlend {blendOption=Linear}
    baseBoneRef {boneName=spine_05}
    componentAppliedLinearAccClamp=<2000 2000 2000>
    componentLinearAccScale=<0.2 0.2 0.3>
    componentLinearVelScale=<0.2 0.2 0.3>
    overrideWorldGravity {z=-980}
   }
   Save cached pose 'SkydiveSimPose'(Pose=Component To Local(ComponentPose=@n48)) {cacheName=SkydiveSimPose}
   Get upper_arm_l_rot_up_alpha() -> (upper_arm_l_rot_up_alpha:double)
  ]}
 EventGraph {nodes [
   Event BlueprintUpdateAnimation()
   > n1 = Sequence!(execute) -> (then_0, then_1, then_2:Exec, then_3:Exec, then_4:Exec, then_5:Exec, then_6:Exec, then_7:Exec)
   > Set skel_mesh(skel_mesh:Skinned Mesh Component Object Reference=GetOwnerSkeletalMeshComponent()) -> (Output_Get:Skinned Mesh Component Object Reference)
   n2 = Get skel_mesh() -> (skel_mesh:Skinned Mesh Component Object Reference)
   n3 = BreakRotator(InRot=BreakTransform(InTransform=GetDeltaTransformFromRefPose(self=@n2, BoneName=head, BaseName=spine_05)).Rotation)
   > Set scarf_front_rot(scarf_front_rot:Rotator=MakeRotator(Roll=@n3, Pitch=@n3.Pitch, Yaw=MapRangeClamped(Value=@n3.Yaw, InRangeA=5, InRangeB=100, OutRangeB=100))) -> (Output_Get:Rotator)
   > Set scarf_back_rot(scarf_back_rot:Rotator=MakeRotator(Roll=@n3, Pitch=@n3.Pitch, Yaw=MapRangeClamped(Value=@n3.Yaw, InRangeA=0, InRangeB=-100, OutRangeB=-100))) -> (Output_Get:Rotator)
   > Set head_back_alpha(head_back_alpha:double=MapRangeClamped(Value=@n3.Yaw, InRangeA=0, InRangeB=-40)) -> (Output_Get:double)
   n4 = BreakTransform(InTransform=GetDeltaTransformFromRefPose(self=Get skel_mesh(), BoneName=clavicle_l, BaseName=spine_05))
   Set upper_arm_l_rot_up_alpha(execute=@n1.then_2, upper_arm_l_rot_up_alpha:double=MapRangeClamped(Value=BreakRotator(InRot=@n4.Rotation).Pitch, InRangeA=0, InRangeB=40)) -> (Output_Get:double)
   n5 = Get skel_mesh()
   n6 = BreakRotator(InRot=BreakTransform(InTransform=GetDeltaTransformFromRefPose(self=@n5, BoneName=clavicle_r, BaseName=spine_05)).Rotation)
   > Set clavicle_r_hood_base_rot(clavicle_r_hood_base_rot:Rotator=MakeRotator(Pitch=@n6.Pitch, Yaw=@n6.Yaw)) -> (Output_Get:Rotator)
   n7 = Get skel_mesh()
   n8 = BreakRotator(InRot=BreakTransform(InTransform=GetDeltaTransformFromRefPose(self=@n7, BoneName=upperarm_r, BaseName=clavicle_r)).Rotation)
   > Set upper_arm_r_rot_down_alpha(upper_arm_r_rot_down_alpha:double=MapRangeClamped(Value=@n8.Pitch, InRangeA=-10, InRangeB=-30)) -> (Output_Get:double)
   > n9 = Branch(Condition=Get bIsPlayingEmote())
   Set upper_arm_r_rot_up_alpha(execute=@n9.else, upper_arm_r_rot_up_alpha:double=MapRangeClamped(Value=@n8.Yaw, InRangeA=50, InRangeB=120)) -> (Output_Get:double)
   > Set upper_arm_r_rot_up_alpha_0(upper_arm_r_rot_up_alpha_0:double=MapRangeClamped(Value=@n8.Yaw, InRangeA=90, InRangeB=120)) -> (Output_Get:double)
   > Set upper_arm_r_rot_in_alpha(upper_arm_r_rot_in_alpha:double=MapRangeClamped(Value=@n8.Yaw, InRangeA=100, InRangeB=120)) -> (Output_Get:double)
   n10 = Gravity Override State Function DEPRECATED(
    execute=@n1.then_3
    CharacterPartAnimInstance=Self-Reference()
    GameTimeInSeconds=GetGameTimeInSeconds()
    GravityOverrideParams=Get Gravity Override Params()
   )
   > Set GravityOverride(GravityOverride=@n10)
   n11 = Make Array("[0]"=(TagName="Cosmetics.Variant.Property.Stage1"))
   Set cape_backpack_equipped(
    execute=[@n19.then, @n1.then_4]
    cape_backpack_equipped:bool=AND Boolean(A=CharacterItemDefinitionHasAnyMetaTag(Tags=@n11), B=CheckForBackbling())
   ) -> (Output_Get:bool)
   n12 = Branch(execute=@n1.then_5, Condition=Get skydive() -> (skydive:bool))
   > Set upper_arm_back_r(upper_arm_back_r:double=0) -> (Output_Get:double)
   > Set upper_arm_back_l(upper_arm_back_l:double=0) -> (Output_Get:double)
   n13 = GetDeltaTransformFromRefPose(self=GetPartSkeletalMeshComponent(), BoneName=upperarm_l, BaseName=spine_05)
   n14 = MapRangeClamped(Value=BreakRotator(InRot=BreakTransform(InTransform=@n13).Rotation).Yaw, InRangeA=10, InRangeB=-30)
   > Set clavicle_glider_float_l(clavicle_glider_float_l:double=@n14) -> (Output_Get:double)
   n15 = BreakTransform(InTransform=GetDeltaTransformFromRefPose(self=GetPartSkeletalMeshComponent(), BoneName=upperarm_r, BaseName=spine_05))
   Set upper_arm_back_r(execute=@n12.else, upper_arm_back_r=MapRangeClamped(Value=BreakRotator(InRot=@n15.Rotation).Yaw, InRangeA=0, InRangeB=-40))
   > Set upper_arm_back_l(upper_arm_back_l=@n14)
   > Set clavicle_glider_float_l(clavicle_glider_float_l=0)
   n16 = BreakTransform(InTransform=GetDeltaTransformFromRefPose(self=GetPartSkeletalMeshComponent(), BoneName=clavicle_l, BaseName=spine_05))
   > Set clavicle_up_l(clavicle_up_l:double=MapRangeClamped(Value=BreakRotator(InRot=@n16.Rotation).Pitch, InRangeA=0, InRangeB=40)) -> (Output_Get:double)
   n17 = BreakTransform(InTransform=GetDeltaTransformFromRefPose(self=GetPartSkeletalMeshComponent(), BoneName=clavicle_r, BaseName=spine_05))
   > Set clavicle_up_r(clavicle_up_r:double=MapRangeClamped(Value=BreakRotator(InRot=@n17.Rotation).Pitch, InRangeA=0, InRangeB=40)) -> (Output_Get:double)
   Set skydive(
    execute=@n1.then_6
    skydive:bool=OR Boolean!(A=Get bIsSkyDiving(), B=Get bIsSkydiveFloating(), C:bool=Get bIsSkydiveDiveMode()) -> (ReturnValue)
   ) -> (Output_Get:bool)
   n18 = GetAnimInstance(self=GetOwnerSkeletalMeshComponent())
   Set EmoteSaluteRBANAlpha(
    execute=@n1.then_7
    EmoteSaluteRBANAlpha:double=SelectFloat(
     A=0.5
     B=1
     bPickA=Montage_IsPlaying(self=@n18, Montage=/BRCosmetics/Animation/Game/MainPlayer/Emotes/ClearRadius/CMF/Emote_ClearRadius_CMF_Cape_M)
    )
   ) -> (Output_Get:double)
   Branch(execute=@n1.then_1, Condition=Get bIsInFrontEnd())
   > Set Custom Idle(Custom Idle:double=GetCurveValue(self=GetOwnerFrontEndAnimInstance(), CurveName=Frontend_Idle_Bypass)) -> (Output_Get:double)
   n19 = Event OnCharacterCustomizationLoaded()
  ]}
]
variables [
 As Fort Player Pawn {type=FortPlayerPawn}
 cape_backpack_equipped {type=boolean}
 clavicle_glider_float_l {type=number}
 clavicle_r_hood_base_rot {type=Rotator}
 clavicle_up_l {type=number}
 clavicle_up_r {type=number}
 Custom Idle {type=number}
 EmoteSaluteRBANAlpha {type=number}
 Gravity Override Params {type=GravityOverrideParamsStruct}
 GravityOverride {type=Vector}
 head_back_alpha {type=number}
 scarf_back_rot {type=Rotator}
 scarf_front_rot {type=Rotator}
 skel_mesh {type=SkinnedMeshComponent, default=null}
 skydive {type=boolean}
 upper_arm_back_l {type=number}
 upper_arm_back_r {type=number}
 upper_arm_l_rot_up_alpha {type=number}
 upper_arm_r_rot_down_alpha {type=number}
 upper_arm_r_rot_in_alpha {type=number}
 upper_arm_r_rot_up_alpha {type=number}
 upper_arm_r_rot_up_alpha_0 {type=number}
]
