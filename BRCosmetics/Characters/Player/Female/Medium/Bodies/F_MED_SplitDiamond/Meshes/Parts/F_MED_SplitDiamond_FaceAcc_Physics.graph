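# Physics asset for the FaceAcc part of the SplitDiamond medium female character, providing capsule bodies and joints for physics-driven hair.
# Seven lettered hair groups (a–g) and four three-bone front strands form constraint chains rooted at a single head capsule.
# Strand segments use tight swing and twist limits, while most master and root helper joints, with mostly tiny capsules, leave rotation free.
# Group b has no master and joins the head through a limited joint.
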PhysicsAsset /BRCosmetics/Characters/Player/Female/Medium/Bodies/F_MED_SplitDiamond/Meshes/Parts/F_MED_SplitDiamond_FaceAcc_Physics
bodies {
 hair_a_01 [{shapeType=Capsule, length=0.505, radius=0.505, shapeName=hair_a_01_capsule}]
 hair_a_02 [{shapeType=Capsule, length=7.5221, radius=2.25, shapeName=hair_a_02_capsule, rotation=<rot 90 0 0>}]
 hair_a_master [{shapeType=Capsule, length=0.505, radius=0.505, shapeName=hair_a_master_capsule}]
 hair_b_01 [{shapeType=Capsule, length=0.505, radius=0.505, shapeName=hair_b_01_capsule, rotation=<rot 0 -90 0>}]
 hair_b_02 [{shapeType=Capsule, length=6.4682, radius=2, shapeName=hair_b_02_capsule, rotation=<rot 90 0 0>}]
 hair_c_00 [{shapeType=Capsule, length=0.505, radius=0.505, shapeName=hair_c_00_capsule}]
 hair_c_01 [{shapeType=Capsule, length=4.255, radius=1.25, shapeName=hair_c_01_capsule, center=<1 0 0>, rotation=<rot 90 0 0>}]
 hair_c_02 [{shapeType=Capsule, length=4.3673, radius=2.25, shapeName=hair_c_02_capsule, center=<1 0 0>, rotation=<rot 90 0 0>}]
 hair_c_master [{shapeType=Capsule, length=0.505, radius=0.505, shapeName=hair_c_master_capsule}]
 hair_d_00 [{shapeType=Capsule, length=0.505, radius=0.505, shapeName=hair_d_00_capsule}]
 hair_d_01 [{shapeType=Capsule, length=4.255, radius=1.25, shapeName=hair_d_01_capsule, center=<1 0 0>, rotation=<rot 90 0 0>}]
 hair_d_02 [{shapeType=Capsule, length=4.255, radius=1.25, shapeName=hair_d_02_capsule, center=<1 0 0>, rotation=<rot 90 0 0>}]
 hair_d_master [{shapeType=Capsule, length=0.505, radius=0.505, shapeName=hair_d_master_capsule}]
 hair_e_01 [{shapeType=Capsule, length=2.255, radius=1.25, shapeName=hair_e_01_capsule, rotation=<rot 90 0 0>}]
 hair_e_02 [{shapeType=Capsule, length=5.1804, radius=1.25, shapeName=hair_e_02_capsule, rotation=<rot 90 0 0>}]
 hair_e_master [{shapeType=Capsule, length=0.505, radius=0.505, shapeName=hair_e_master_capsule}]
 hair_f_01 [{shapeType=Capsule, length=2.505, radius=3, shapeName=hair_f_01_capsule, center=<-1 0 0>, rotation=<rot 90 0 0>}]
 hair_f_02 [{shapeType=Capsule, length=6.0544, radius=2, shapeName=hair_f_02_capsule, center=<-1 0 0>, rotation=<rot 90 0 0>}]
 hair_f_master [{shapeType=Capsule, length=0.505, radius=0.505, shapeName=hair_f_master_capsule}]
 hair_front_a_01 [{shapeType=Capsule, length=2, radius=2, shapeName=hair_front_a_01_capsule, rotation=<rot 90 0 0>}]
 hair_front_a_02 [{shapeType=Capsule, length=2, radius=2, shapeName=hair_front_a_02_capsule, rotation=<rot 90 0 0>}]
 hair_front_a_03 [{shapeType=Capsule, length=1.25, radius=2.75, shapeName=hair_front_a_03_capsule, rotation=<rot 90 0 0>}]
 hair_front_b_01 [{shapeType=Capsule, length=2, radius=2, shapeName=hair_front_b_01_capsule, rotation=<rot 90 0 0>}]
 hair_front_b_02 [{shapeType=Capsule, length=2, radius=2, shapeName=hair_front_b_02_capsule, rotation=<rot 90 0 0>}]
 hair_front_b_03 [{shapeType=Capsule, length=2, radius=2, shapeName=hair_front_b_03_capsule, rotation=<rot 90 0 0>}]
 hair_front_c_01 [{shapeType=Capsule, length=2, radius=2, shapeName=hair_front_c_01_capsule, rotation=<rot 90 0 0>}]
 hair_front_c_02 [{shapeType=Capsule, length=2, radius=2, shapeName=hair_front_c_02_capsule, rotation=<rot 90 0 0>}]
 hair_front_c_03 [{shapeType=Capsule, length=2, radius=2, shapeName=hair_front_c_03_capsule, rotation=<rot 90 0 0>}]
 hair_front_d_01 [{shapeType=Capsule, length=3.25, radius=2, shapeName=hair_front_d_01_capsule, rotation=<rot 90 0 0>}]
 hair_front_d_02 [{shapeType=Capsule, length=3.25, radius=2, shapeName=hair_front_d_02_capsule, rotation=<rot 90 0 0>}]
 hair_front_d_03 [{shapeType=Capsule, length=3.25, radius=2, shapeName=hair_front_d_03_capsule, rotation=<rot 90 0 0>}]
 hair_g_00 [{
   shapeType=Capsule
   length=0.505
   radius=0.505
   shapeName=hair_g_00_capsule
   center=<-0.6874 1.5551 0.3429>
   rotation=<rot -1.6156 -90 12.462>
  }]
 hair_g_01 [{shapeType=Capsule, length=3.755, radius=2.5, shapeName=hair_g_01_capsule, center=<-1 0 0>, rotation=<rot 90 0 0>}]
 hair_g_02 [{shapeType=Capsule, length=4.2267, radius=2.25, shapeName=hair_g_02_capsule, center=<-1 0 0>, rotation=<rot 90 0 0>}]
 hair_g_master [{shapeType=Capsule, length=0.505, radius=0.505, shapeName=hair_g_master_capsule, rotation=<rot 0 -90 0>}]
 head [{shapeType=Capsule, length=2.5, radius=5, center=<0 5.3008 0>, rotation=<rot -90 0 0>}]
}
constraints [
 {bone1Name=hair_a_02, bone2Name=hair_a_01, swing1LimitDegrees=10, swing2LimitDegrees=5, twistLimitDegrees=5}
 {bone1Name=hair_b_02, bone2Name=hair_b_01, swing1LimitDegrees=10, swing2LimitDegrees=5, twistLimitDegrees=5}
 {bone1Name=hair_c_01, bone2Name=hair_c_00, swing1LimitDegrees=15, swing2LimitDegrees=20, twistLimitDegrees=5}
 {bone1Name=hair_c_02, bone2Name=hair_c_01, swing1LimitDegrees=15, swing2LimitDegrees=20, twistLimitDegrees=5}
 {bone1Name=hair_d_01, bone2Name=hair_d_00, swing1LimitDegrees=10, swing2LimitDegrees=20, twistLimitDegrees=5}
 {bone1Name=hair_d_02, bone2Name=hair_d_01, swing1LimitDegrees=10, swing2LimitDegrees=20, twistLimitDegrees=5}
 {bone1Name=hair_e_01, bone2Name=hair_e_master, swing1LimitDegrees=10, swing2LimitDegrees=10, twistLimitDegrees=5}
 {bone1Name=hair_e_02, bone2Name=hair_e_01, swing1LimitDegrees=10, swing2LimitDegrees=10, twistLimitDegrees=5}
 {bone1Name=hair_f_01, bone2Name=hair_f_master, swing1LimitDegrees=20, swing2LimitDegrees=15, twistLimitDegrees=5}
 {bone1Name=hair_f_02, bone2Name=hair_f_01, swing1LimitDegrees=20, swing2LimitDegrees=15, twistLimitDegrees=5}
 {bone1Name=hair_front_a_01, bone2Name=head, swing1LimitDegrees=5, swing2LimitDegrees=5, twistLimitDegrees=5}
 {bone1Name=hair_front_a_02, bone2Name=hair_front_a_01, swing1LimitDegrees=5, swing2LimitDegrees=5, twistLimitDegrees=5}
 {bone1Name=hair_front_a_03, bone2Name=hair_front_a_02, swing1LimitDegrees=5, swing2LimitDegrees=5, twistLimitDegrees=5}
 {bone1Name=hair_front_b_01, bone2Name=head, swing1LimitDegrees=5, swing2LimitDegrees=5, twistLimitDegrees=5}
 {bone1Name=hair_front_b_02, bone2Name=hair_front_b_01, swing1LimitDegrees=5, swing2LimitDegrees=5, twistLimitDegrees=5}
 {bone1Name=hair_front_b_03, bone2Name=hair_front_b_02, swing1LimitDegrees=5, swing2LimitDegrees=5, twistLimitDegrees=5}
 {bone1Name=hair_front_c_01, bone2Name=head, swing1LimitDegrees=5, swing2LimitDegrees=5, twistLimitDegrees=5}
 {bone1Name=hair_front_c_02, bone2Name=hair_front_c_01, swing1LimitDegrees=5, swing2LimitDegrees=5, twistLimitDegrees=5}
 {bone1Name=hair_front_c_03, bone2Name=hair_front_c_02, swing1LimitDegrees=5, swing2LimitDegrees=5, twistLimitDegrees=5}
 {bone1Name=hair_front_d_01, bone2Name=head, swing1LimitDegrees=5, swing2LimitDegrees=5, twistLimitDegrees=5}
 {bone1Name=hair_front_d_02, bone2Name=hair_front_d_01, swing1LimitDegrees=10, swing2LimitDegrees=5, twistLimitDegrees=5}
 {bone1Name=hair_front_d_03, bone2Name=hair_front_d_02, swing1LimitDegrees=10, swing2LimitDegrees=5, twistLimitDegrees=5}
 {bone1Name=hair_g_01, bone2Name=hair_g_00, swing1LimitDegrees=20, swing2LimitDegrees=20, twistLimitDegrees=5}
 {bone1Name=hair_g_02, bone2Name=hair_g_01, swing1LimitDegrees=20, swing2LimitDegrees=20, twistLimitDegrees=5}
 {bone1Name=hair_b_01, bone2Name=head, swing1LimitDegrees=10, swing2LimitDegrees=10, twistLimitDegrees=5}
 {
  bone1Name=hair_a_01
  bone2Name=hair_a_master
  swing1LimitDegrees=45
  swing1Motion=Free
  swing2LimitDegrees=45
  swing2Motion=Free
  twistLimitDegrees=45
  twistMotion=Free
 }
 {
  bone1Name=hair_a_master
  bone2Name=head
  swing1LimitDegrees=45
  swing1Motion=Free
  swing2LimitDegrees=45
  swing2Motion=Free
  twistLimitDegrees=45
  twistMotion=Free
 }
 {
  bone1Name=hair_c_master
  bone2Name=head
  swing1LimitDegrees=45
  swing1Motion=Free
  swing2LimitDegrees=45
  swing2Motion=Free
  twistLimitDegrees=45
  twistMotion=Free
 }
 {
  bone1Name=hair_c_00
  bone2Name=hair_c_master
  swing1LimitDegrees=45
  swing1Motion=Free
  swing2LimitDegrees=45
  swing2Motion=Free
  twistLimitDegrees=45
  twistMotion=Free
 }
 {
  bone1Name=hair_d_master
  bone2Name=head
  swing1LimitDegrees=45
  swing1Motion=Free
  swing2LimitDegrees=45
  swing2Motion=Free
  twistLimitDegrees=45
  twistMotion=Free
 }
 {
  bone1Name=hair_d_00
  bone2Name=hair_d_master
  swing1LimitDegrees=45
  swing1Motion=Free
  swing2LimitDegrees=45
  swing2Motion=Free
  twistLimitDegrees=45
  twistMotion=Free
 }
 {
  bone1Name=hair_e_master
  bone2Name=head
  swing1LimitDegrees=45
  swing1Motion=Free
  swing2LimitDegrees=45
  swing2Motion=Free
  twistLimitDegrees=45
  twistMotion=Free
 }
 {
  bone1Name=hair_f_master
  bone2Name=head
  swing1LimitDegrees=45
  swing1Motion=Free
  swing2LimitDegrees=45
  swing2Motion=Free
  twistLimitDegrees=45
  twistMotion=Free
 }
 {
  bone1Name=hair_g_master
  bone2Name=head
  swing1LimitDegrees=45
  swing1Motion=Free
  swing2LimitDegrees=45
  swing2Motion=Free
  twistLimitDegrees=45
  twistMotion=Free
 }
 {
  bone1Name=hair_g_00
  bone2Name=hair_g_master
  swing1LimitDegrees=45
  swing1Motion=Free
  swing2LimitDegrees=45
  swing2Motion=Free
  twistLimitDegrees=45
  twistMotion=Free
 }
]
