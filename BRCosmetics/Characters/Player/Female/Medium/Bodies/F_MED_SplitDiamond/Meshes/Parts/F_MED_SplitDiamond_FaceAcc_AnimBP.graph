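# Animation blueprint for the SplitDiamond female face accessory, copying the attached parent's pose and simulating it with head-relative RigidBody physics.
# A half-second blend picks the default physics when the cape variant is equipped or the character crouches, otherwise the CapeOff physics.
# The CapeOff branch first rotates three hair bones additively, weighted by head yaw relative to spine_05's ref pose.
# Each update recomputes that weight and cape state from cosmetic meta tags, and calls CheckForBackbling.
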AnimBlueprint /BRCosmetics/Characters/Player/Female/Medium/Bodies/F_MED_SplitDiamond/Meshes/Parts/F_MED_SplitDiamond_FaceAcc_AnimBP
target_skeleton=&/F_MED_SplitDiamond_FaceAcc_Skeleton
graphs [
 AnimGraph {nodes [
   n1 = RigidBody(ComponentPose=Local To Component(LocalPose=Copy Pose From Mesh() {bUseAttachedParent=true})) {
    bOverrideWorldGravity=true
    overridePhysicsAsset=&/F_MED_SplitDiamond_FaceAcc_Physics
    simulationSpace=BaseBoneSpace
    alphaBoolBlend {blendOption=Linear}
    baseBoneRef {boneName=head}
    componentAppliedLinearAccClamp=<2250 2250 2250>
    componentLinearAccScale=<0.35 0.35 0.4>
    componentLinearVelScale=<0.35 0.35 0.4>
    overrideWorldGravity {z=-980}
   }
   n2 = Get head_back_alpha() -> (head_back_alpha:double)
   n3 = "Transform (Modify) Bone - Bone: hair_c_00"(
    ComponentPose=Local To Component(LocalPose=Copy Pose From Mesh() {bUseAttachedParent=true})
    Alpha=@n2
    Translation="0,0,0"
    Rotation="0,-30,0"
    Scale="1,1,1"
   ) {
    rotationMode=BMM_Additive
    rotationSpace=BCS_BoneSpace
    alphaBoolBlend {blendOption=Linear}
    boneToModify {boneName=hair_c_00}
    rotation {yaw=-30}
   }
   n4 = "Transform (Modify) Bone - Bone: hair_d_00"(ComponentPose=@n3, Alpha=@n2, Translation="0,0,0", Rotation="0,-45,0", Scale="1,1,1") {
    rotationMode=BMM_Additive
    rotationSpace=BCS_BoneSpace
    alphaBoolBlend {blendOption=Linear}
    boneToModify {boneName=hair_d_00}
    rotation {yaw=-45}
   }
   n5 = "Transform (Modify) Bone - Bone: hair_g_00"(ComponentPose=@n4, Alpha=@n2, Translation="0,0,0", Rotation="-25,-45,0", Scale="1,1,1") {
    rotationMode=BMM_Additive
    rotationSpace=BCS_BoneSpace
    alphaBoolBlend {blendOption=Linear}
    boneToModify {boneName=hair_g_00}
    rotation {pitch=-25, yaw=-45}
   }
   n6 = RigidBody(ComponentPose=@n5) {
    bOverrideWorldGravity=true
    overridePhysicsAsset=&/F_MED_SplitDiamond_CapeOff_FaceAcc_Physics
    simulationSpace=BaseBoneSpace
    alphaBoolBlend {blendOption=Linear}
    baseBoneRef {boneName=head}
    componentAppliedLinearAccClamp=<3000 3000 3000>
    componentLinearAccScale=<0.5 0.5 0.6>
    componentLinearVelScale=<0.5 0.5 0.6>
    overrideWorldGravity {z=-980}
   }
   Output Pose(Result=Blend Poses by bool(
     BlendPose_0=Component To Local(ComponentPose=@n1)
     BlendPose_1=Component To Local(ComponentPose=@n6)
     BlendTime_0=0.5
     BlendTime_1=0.5
     bActiveValue=Get cape_equipped_crouching() -> (cape_equipped_crouching:bool)
    ) {blendType=Linear}) {layerGroup=None}
  ]}
 EventGraph {nodes [
   Event BlueprintUpdateAnimation()
   > n1 = Sequence!(execute) -> (then_0, then_1, then_2:Exec)
   > Set skel_mesh(skel_mesh:Skinned Mesh Component Object Reference=GetOwnerSkeletalMeshComponent()) -> (Output_Get:Skinned Mesh Component Object Reference)
   n2 = BreakTransform(InTransform=GetDeltaTransformFromRefPose(self=Get skel_mesh() -> (skel_mesh:Skinned Mesh Component Object Reference), BoneName=head, BaseName=spine_05))
   > Set head_back_alpha(head_back_alpha:double=MapRangeClamped(Value=BreakRotator(InRot=@n2.Rotation).Yaw, InRangeA=0, InRangeB=-40)) -> (Output_Get:double)
   n3 = Set cape_equipped(
    execute=@n1.then_1
    cape_equipped:bool=CharacterItemDefinitionHasAnyMetaTag(Tags=Make Array("[0]"=(TagName="Cosmetics.Variant.Property.Stage1")))
   ) -> (Output_Get:bool)
   > Set cape_equipped_crouching(cape_equipped_crouching:bool=OR Boolean(A=@n3, B=Get bIsCrouching())) -> (Output_Get:bool)
   CheckForBackbling()
  ]}
]
variables [
 cape_equipped {type=boolean}
 cape_equipped_crouching {type=boolean}
 head_back_alpha {type=number}
 skel_mesh {type=SkinnedMeshComponent, default=null}
]
